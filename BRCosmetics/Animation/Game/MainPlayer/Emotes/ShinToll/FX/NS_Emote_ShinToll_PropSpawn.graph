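# Niagara one-shot effect for the ShinToll emote's prop spawn: a brief pink and gold heart pop with burst, ring, sparks, twinkling sparkles and light.
# NE_Burst (sprites plus mesh) and NE_Heart layer several renderers, using spawn group to set VisibilityTag and switch color and size.
# Sparks stretch with speed under drag and gravity, while sparkles flicker through waveform and color curves.
# User.bIsFrontEndPreview applies FrontEndSize to burst scaling only.
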NiagaraSystem /BRCosmetics/Animation/Game/MainPlayer/Emotes/ShinToll/FX/NS_Emote_ShinToll_PropSpawn
dependencies {
 usedDataInterfaces [
  &NiagaraDataInterfaceCurve
  &NiagaraDataInterfaceMeshRendererInfo
  &NiagaraDataInterfaceColorCurve
  &NiagaraDataInterfaceArrayColor
 ]
 usedDynamicInputs [&FN_FrontEndVector3Modifier]
 usedModules [&SystemState]
}
emitters {
 NE_Burst {
  emitterUpdateScript [
   &EmitterState(Inactive Response=::NewEnumerator1, Life Cycle Mode=::NewEnumerator1, Loop Behavior=::NewEnumerator1, Loop Duration=1, UseLoopDelay=false)
   &SpawnBurst_Instantaneous(Use Spawn Probability=false, UseLoopCountLimit=false)
   &SpawnBurst_Instantaneous(Spawn Group=1, Use Spawn Probability=false, UseLoopCountLimit=false) {module=SpawnBurst_Instantaneous001}
   &SpawnBurst_Instantaneous(Spawn Group=2, Use Spawn Probability=false, UseLoopCountLimit=false) {module=SpawnBurst_Instantaneous002}
  ]
  particleSpawnScript [
   /SetVariables(Particles.VisibilityTag=&SpawnGroup()) {isSetParameters=true}
   &InitializeParticle(
    Color=&MergeHelper_ColorSwitch(
     0=<rgba 2 1.8415 0.7159 0.4>
     1=<rgba 1 1 1 1>
     2=<rgba 0.1536 0.0459 0.0282 0.6652>
     3=?
     Index=&SpawnGroup()
     Num=::NewEnumerator1
    )
    Lifetime=&MergeHelper_FloatSwitch(0=0.1, 1=0.1, 2=0.2, 3=?, Index=&SpawnGroup(), Num=::NewEnumerator1)
    Mesh Scale=<1.2 1.2 1.3>
    Mesh Scale Mode=::NewEnumerator4
    Position Offset=<0 20 0>
    Sprite Size Mode=::NewEnumerator3
    Uniform Sprite Size=&MergeHelper_FloatSwitch(0=500, 1=0, 2=70, 3=?, Index=&SpawnGroup(), Num=::NewEnumerator1)
    UsePositionOffset=true
    Write Material Random=true
   )
   &InitialMeshOrientation(Mesh Orientation Mode=::NewEnumerator2, Rotation=<0 0.15 -0.032>, Use Rotation Vector=true)
   &CameraOffset(Camera Offset Amount=10) {module=CameraOffset001}
  ]
  particleUpdateScript [
   &ParticleState()
   &ScaleSpriteSize(
    Uniform Curve Scale=&MakeCustomFloatFromBool(Bool=$User.bIsFrontEndPreview, True Float=$User.FrontEndSize, False Float=1)
    Uniform Curve Sprite Scale=DataInterface() {propertyValues {Curve {keys [<at 0 1>, {tangentMode=RCTM_User, time=1, value=0.2}]}}}
   )
   /SetVariables(Particles.Size002:Vector2f=&Multiply_Vector2D(A=$Particles.SpriteSize, B=&Vector2DFromFloat(Value=0.65))) {isSetParameters=true}
   &ScaleMeshSize(Scale Factor=&VectorFromFloat(Value=&FloatFromCurve(
      FloatCurve=DataInterface() {propertyValues {Curve {keys [
          {arriveTangent=3.9833, interpMode=RCIM_Cubic, leaveTangent=3.9833, tangentMode=RCTM_User, time=0, value=0.2}
          {arriveTangent=5.6067, tangentMode=RCTM_User, time=0.15, value=1}
          {arriveTangent=0.0139, interpMode=RCIM_Cubic, leaveTangent=0.0139, tangentMode=RCTM_User, time=1, value=0}
         ]}}}
      Scale Curve=&MakeCustomFloatFromBool(Bool=$User.bIsFrontEndPreview, True Float=$User.FrontEndSize, False Float=1)
     )))
   &ScaleColor(
    Scale Alpha=&FloatFromCurve(FloatCurve=DataInterface() {propertyValues {Curve {keys [<at 0 1>,<at 0.2 1>,<at 1 0>]}}})
    ScaleA=true
    ScaleRGB=true
   )
   &DynamicMaterialParameters(Module.Index0Float0Write=true, Module.Index0Float1Write=true, Module.Index0Float2Write=true, Module.Index0Float3Write=true) {module=DynamicMaterialParameters001}
  ]
  rendererClasses [&NiagaraSpriteRendererProperties, &NiagaraMeshRendererProperties]
  renderers [
   {rendererClass=&NiagaraSpriteRendererProperties}
   {rendererClass=&NiagaraMeshRendererProperties, rendererIndex=1}
   {rendererClass=&NiagaraSpriteRendererProperties, rendererIndex=2}
  ]
 }
 NE_Heart {
  emitterUpdateScript [
   &EmitterState()
   &SpawnBurst_Instantaneous(Spawn Count=1, Spawn Time=0.02, Use Spawn Probability=false)
   &SpawnBurst_Instantaneous(Spawn Count=1, Spawn Group=1, Spawn Time=0.02, Use Spawn Probability=false) {module=SpawnBurst_Instantaneous001}
   &SpawnBurst_Instantaneous(Spawn Count=1, Spawn Group=2, Spawn Time=0.02, Use Spawn Probability=false) {module=SpawnBurst_Instantaneous002}
   &SpawnBurst_Instantaneous(Spawn Count=1, Spawn Group=3, Spawn Time=0.02, Use Spawn Probability=false) {module=SpawnBurst_Instantaneous003}
   &SpawnBurst_Instantaneous(Spawn Count=1, Spawn Group=4, Spawn Time=0.02, Use Spawn Probability=false) {module=SpawnBurst_Instantaneous004}
   &SpawnBurst_Instantaneous(Spawn Count=1, Spawn Group=5, Spawn Time=0.02, Use Spawn Probability=false) {module=SpawnBurst_Instantaneous005}
   &SpawnBurst_Instantaneous(Spawn Count=1, Spawn Group=6, Spawn Time=0.02, Use Spawn Probability=false) {module=SpawnBurst_Instantaneous006}
  ]
  particleSpawnScript [
   /SetVariables(Particles.VisibilityTag=&SpawnGroup()) {isSetParameters=true}
   &InitializeParticle(
    Color=&MergeHelper_ColorSwitch(
     0=<rgba 4 0.3333 2.0479 0.2>
     1=<rgba 6 1.2031 3.3677 0.2>
     2=<rgba 8 2.4631 4.8916 0.15>
     3=<rgba 3 1.3337 1.9817 0.6739>
     4=<rgba 5 0.4688 1.9205 0.2174>
     5=<rgba 5 0 1.3773 0.15>
     6=<rgba 0.0677 0.0153 0.0402 1.4>
     Index=&SpawnGroup()
     Num=::NewEnumerator5
    )
    Lifetime=&MergeHelper_FloatSwitch(0=0.45, 1=0.45, 2=0.45, 3=0.5, 4=0.45, 5=0.5, 6=0.65, Index=&SpawnGroup(), Num=::NewEnumerator5)
    Sprite Size=&MergeHelper_Vector2Switch(
     0=<230 200>
     1=&Multiply_Vector2D_ByFloat(Float=0.75, Vector2D=<230 200>)
     2=&Multiply_Vector2D_ByFloat(Float=0.5, Vector2D=<230 200>)
     3=<240 30>
     4=&Multiply_Vector2D_ByFloat(Float=1.1, Vector2D=<230 200>)
     5=<200 200>
     6=<310 310>
     Index=&SpawnGroup()
     Num=::NewEnumerator5
    )
    Sprite Size Mode=::NewEnumerator4
    UsePositionOffset=false
    Write Material Random=true
   )
  ]
  particleUpdateScript [
   &ParticleState()
   &ScaleSpriteSize(
    Uniform Curve Scale=0.6
    Uniform Curve Sprite Scale=DataInterface() {propertyValues {Curve {keys [{tangentMode=RCTM_User, time=0, value=0}, <at 0.2 0.8>, {tangentMode=RCTM_User, time=1, value=1}]}}}
   )
   /SetVariables(Particles.SpriteSize02:Vector2f=&Multiply_Vector2D_ByFloat(
     Float=&FloatFromCurve(
      FloatCurve=DataInterface() {propertyValues {Curve {keys [{tangentMode=RCTM_User, time=0, value=0}, <at 0.2 0.8>, {tangentMode=RCTM_User, time=1, value=1}]}}}
      Scale Curve=1.1
     )
     Vector2D=&MergeHelper_Vector2Switch(
      0=<0 0>
      1=<0 0>
      2=<0 0>
      3=<0 0>
      4=<0 0>
      5=<200 200>
      6=<310 310>
      Index=&SpawnGroup()
      Num=::NewEnumerator5
     )
    )) {isSetParameters=true}
   &ScaleColor(
    Scale Alpha=&FloatFromCurve(FloatCurve=DataInterface() {propertyValues {Curve {keys [<at 0 1>,<at 0.5 1>,<at 1 0>]}}})
    ScaleA=true
    ScaleRGB=true
   )
   &SpriteFacingAndAlignment(Sprite Facing=<0 1 0>, WriteAlignment=true, WriteFacing=true)
   &SolveForcesAndVelocity(Clamp Velocity=false, Limit Acceleration=false)
  ]
  rendererClasses [&NiagaraSpriteRendererProperties]
  renderers [
   {rendererClass=&NiagaraSpriteRendererProperties}
   {rendererClass=&NiagaraSpriteRendererProperties, rendererIndex=1}
   {rendererClass=&NiagaraSpriteRendererProperties, rendererIndex=2}
   {rendererClass=&NiagaraSpriteRendererProperties, rendererIndex=3}
   {rendererClass=&NiagaraSpriteRendererProperties, rendererIndex=4}
   {rendererClass=&NiagaraSpriteRendererProperties, rendererIndex=5}
   {rendererClass=&NiagaraSpriteRendererProperties, rendererIndex=6}
  ]
 }
 NE_Light {
  emitterUpdateScript [
   &EmitterState()
   &SpawnBurst_Instantaneous(Spawn Count=1, Spawn Time=0.02, Use Spawn Probability=false)
  ]
  particleSpawnScript [&InitializeParticle(Lifetime=0.45, Sprite Size Mode=::NewEnumerator3, Uniform Sprite Size=230, UsePositionOffset=false, Write Material Random=true)]
  particleUpdateScript [
   &ParticleState()
   &SolveForcesAndVelocity(Clamp Velocity=false, Limit Acceleration=false)
   &FN_Light_Attributes(
    Color=&ColorFromCurve(ColorCurve=DataInterface() {propertyValues {AlphaCurve {keys []}, BlueCurve {keys [<at 0.098 2.5955>]}, GreenCurve {keys [<at 0.098 0.5625>]}, RedCurve {keys [<at 0.098 3>]}}})
    Create Color Attribute=true
    Create Exponent Attribute=true
    Create Radius Attribute=true
    Exponent=1
    Radius=&FloatFromCurve(
     FloatCurve=DataInterface() {propertyValues {Curve {defaultValue=0, keys [{tangentMode=RCTM_User, time=0, value=1}, <at 0.2532 1>, {tangentMode=RCTM_User, time=1, value=0}]}}}
     Scale Curve=&MakeCustomFloatFromBool(
      Bool=&Boolean_OrOperation(Bool1=$User.bIsFrontEnd, Bool2=$User.bIsFrontEndPreview)
      True Float=130
      False Float=130
     )
    )
    Scale Color=&FloatFromCurve(
     FloatCurve=DataInterface() {propertyValues {Curve {defaultValue=1, keys [{tangentMode=RCTM_User, time=0, value=1}, {tangentMode=RCTM_User, time=1, value=0}]}}}
     Scale Curve=10
    )
   )
  ]
  rendererClasses [&NiagaraLightRendererProperties]
  renderers [{rendererClass=&NiagaraLightRendererProperties, rendererIndex=0}]
 }
 NE_Particles {
  emitterUpdateScript [
   &EmitterState(Inactive Response=::NewEnumerator1, Life Cycle Mode=::NewEnumerator1, Loop Behavior=::NewEnumerator1, Loop Duration=1, UseLoopDelay=false)
   &SpawnBurst_Instantaneous(Spawn Count=30, Spawn Time=0.05, Use Spawn Probability=false, UseLoopCountLimit=false)
  ]
  particleSpawnScript [
   /SetVariables(
    Particles.SparkIntensity:NiagaraFloat=&FloatFromCurve(
     FloatCurve=DataInterface() {propertyValues {Curve {keys [
         <at 0 1 RCIM_Cubic>
         {arriveTangent=-2.5317, interpMode=RCIM_Cubic, leaveTangent=-2.5317, tangentMode=RCTM_User, time=0.6, value=0.5}
         <at 1 0 RCIM_Cubic>
        ]}}}
     CurveIndex=&RandomRangeFloat(Minimum=0, Maximum=1)
    )
    Particles.VisibilityTag=&SpawnGroup()
   ) {isSetParameters=true}
   &InitializeParticle(
    Color=&LerpLinearColors(StartColor=<rgba 2.5 2.3165 0.6473 1>, EndColor=<rgba 3 0.4071 2.769 1>, LerpFactor=$Particles.SparkIntensity)
    Lifetime Max=0.5
    Lifetime Min=0.2
    Lifetime Mode=::NewEnumerator1
    Sprite Size Max=<3 8>
    Sprite Size Min=<2 6>
    Sprite Size Mode=::NewEnumerator2
    UsePositionOffset=false
    Write Material Random=true
   )
   &ShapeLocation(MaskSpawnGroup=false, Override Seed=false, Sphere Radius=30, Sphere Surface Distribution=1, UseOwnerScale=true)
   &AddVelocityFromPoint(Module.Use Falloff Distance=false, Velocity Strength=&RandomRangeFloat(Minimum=150, Maximum=300))
   &CameraOffset(Camera Offset Amount=-30)
  ]
  particleUpdateScript [
   &ParticleState()
   &SpriteRotationRate(Rotation Rate=&Multiply_Float(
     A=&VectorLength(Vector=$Particles.Velocity)
     B=&RandomRangeFloat(Minimum=-5, Maximum=5)
    ))
   &ScaleColor(
    Linear Color Curve=DataInterface() {propertyValues {
      AlphaCurve {keys [<at 0 1>,<at 1 1>]}
      BlueCurve {keys [
        <at 0 2 RCIM_Cubic>
        {arriveTangent=-1.452, interpMode=RCIM_Cubic, leaveTangent=-1.452, tangentMode=RCTM_User, time=0.163, value=1}
        <at 1 0.0755>
       ]}
      GreenCurve {keys [
        <at 0 2 RCIM_Cubic>
        {arriveTangent=-1.4408, interpMode=RCIM_Cubic, leaveTangent=-1.4408, tangentMode=RCTM_User, time=0.163, value=1}
        <at 1 0.078>
       ]}
      RedCurve {keys [
        <at 0 2 RCIM_Cubic>
        {arriveTangent=-0.8332, interpMode=RCIM_Cubic, leaveTangent=-0.8332, tangentMode=RCTM_User, time=0.163, value=1}
        <at 1 0.25>
       ]}
     }}
    Scale Mode=::NewEnumerator2
   ) {module=ScaleColor001}
   &Drag(Drag=&UniformRangedFloat(Minimum=2, Maximum=3), Use Linear Drag=true, Use Rotational Drag=false)
   &ScaleSpriteSize(Scale Factor=&Vector2DFromFloat(Value=&FloatFromCurve(FloatCurve=DataInterface() {propertyValues {Curve {keys [
          {arriveTangent=-0.8333, interpMode=RCIM_Cubic, time=0, value=0}
          {arriveTangentWeight=0.1795, interpMode=RCIM_Cubic, leaveTangentWeight=0.1795, tangentMode=RCTM_User, time=0.2, value=1}
          {arriveTangent=-0.625, interpMode=RCIM_Cubic, leaveTangent=-0.625, time=0.4, value=1}
          {arriveTangent=-2.2814, interpMode=RCIM_Cubic, leaveTangent=-2.2814, tangentMode=RCTM_User, time=1, value=0.5}
         ]}}})))
   &GravityForce(Gravity=<0 0 -100>)
   &SolveForcesAndVelocity(Clamp Velocity=false, Limit Acceleration=false)
   &ScaleSpriteSizeBySpeed(Max Scale Factor=<0.5 6>, Min Scale Factor=<0.8 1>)
  ]
  rendererClasses [&NiagaraSpriteRendererProperties]
  renderers [{rendererClass=&NiagaraSpriteRendererProperties}]
 }
 NE_Ring {
  emitterUpdateScript [
   &EmitterState()
   &SpawnBurst_Instantaneous(Spawn Count=1, Spawn Time=0.01, Use Spawn Probability=false)
  ]
  particleSpawnScript [
   &InitializeParticle(
    Color=<rgba 10 10 10 1>
    Lifetime=0.2
    Sprite Rotation Angle=264
    Sprite Rotation Mode=::NewEnumerator0
    Sprite Size Mode=::NewEnumerator3
    Uniform Sprite Size=180
    UsePositionOffset=false
    Write Material Random=true
   )
   &CameraOffset(Camera Offset Amount=-30)
  ]
  particleUpdateScript [
   &ParticleState()
   &SpriteRotationRate(Rotation Rate=40)
   &SolveForcesAndVelocity(Clamp Velocity=false, Limit Acceleration=false)
   &ScaleSpriteSize(Scale Factor=&Vector2DFromFloat(Value=&FloatFromCurve(FloatCurve=DataInterface() {propertyValues {Curve {defaultValue=0, keys [{tangentMode=RCTM_User, time=0, value=0}, <at 0.2 0.45>, {tangentMode=RCTM_User, time=1, value=1}]}}})))
   /SetVariables(Particles.SpriteSize02:Vector2f=&Multiply_Vector2D_ByFloat(Float=1, Vector2D=$Particles.SpriteSize)) {isSetParameters=true}
   /SetVariables(Particles.Color02:LinearColor=&Multiply_LinearColorByFloat(Linear Color=$Particles.Color, Float=1.1)) {isSetParameters=true}
   &ScaleColor(
    Linear Color Curve=DataInterface() {propertyValues {
      AlphaCurve {keys [<at 0 1>,<at 0.9981 0>]}
      BlueCurve {keys [<at 0.5507 0.6576>,<at 0.6807 0.3288>]}
      GreenCurve {keys [<at 0.5507 1.9227>,<at 0.6807 0.9614>]}
      RedCurve {keys [<at 0.5507 2>,<at 0.6807 1>]}
     }}
    Scale Mode=2
   )
   &DynamicMaterialParameters(
    Index 0 Param 1=1
    Index 0 Param 2=1
    Index 0 Param 4=&FloatFromCurve(FloatCurve=DataInterface() {propertyValues {Curve {
        defaultValue=0.9
        keys [
         {interpMode=RCIM_Cubic, tangentMode=RCTM_User, time=0.5, value=0.7}
         {arriveTangent=-0.9886, interpMode=RCIM_Cubic, leaveTangent=-0.9886, tangentMode=RCTM_User, time=1, value=0.3}
        ]
       }}})
    Module.Index0Float0Write=true
    Module.Index0Float1Write=true
    Module.Index0Float2Write=true
    Module.Index0Float3Write=true
   )
  ]
  rendererClasses [&NiagaraSpriteRendererProperties]
  renderers [
   {rendererClass=&NiagaraSpriteRendererProperties}
   {rendererClass=&NiagaraSpriteRendererProperties, rendererIndex=1}
  ]
 }
 NE_Sparkles {
  emitterUpdateScript [
   &EmitterState(Life Cycle Mode=::NewEnumerator1, Loop Duration=2, UseLoopDelay=false)
   &SpawnBurst_Instantaneous(Spawn Count=10, Spawn Time=0.05, Use Spawn Probability=false, UseLoopCountLimit=false)
  ]
  particleSpawnScript [
   &InitializeParticle(
    Lifetime Max=0.8
    Lifetime Min=0.3
    Lifetime Mode=::NewEnumerator1
    Mass Mode=::NewEnumerator3
    Sprite Rotation Angle=45
    Sprite Rotation Mode=::NewEnumerator0
    Sprite Size Mode=::NewEnumerator3
    Uniform Sprite Size=&RandomRangeFloat(Minimum=10, Maximum=24)
    UsePositionOffset=false
    Write Material Random=true
   )
   &ShapeLocation(MaskSpawnGroup=false, Override Seed=false, Sphere Radius=30, Sphere Surface Distribution=1, UseOwnerScale=true)
   &AddVelocity(
    Module.Use Falloff Distance=false
    Override Seed=false
    Velocity Mode=::NewEnumerator1
    Velocity Speed=&RandomRangeFloat(Minimum=100, Maximum=200)
   )
   &CameraOffset(Camera Offset Amount=-30)
  ]
  particleUpdateScript [
   &ParticleState()
   &SpriteRotationRate(Rotation Rate=&RandomRangeFloat(Minimum=100, Maximum=100))
   &Drag(Drag=&RandomRangeFloat(Minimum=2.5, Maximum=3.5), Use Linear Drag=true, Use Rotational Drag=false)
   &ScaleColor(
    Scale Alpha=&FloatFromCurve(
     FloatCurve=DataInterface() {propertyValues {Curve {keys [<at 0 1 RCIM_Cubic>,<at 1.0011 0.002 RCIM_Cubic>]}}}
     Scale Curve=&Multiply_Float(
      A=1
      B=&Waveform(
       "Waveform [1]"=::NewEnumerator0
       "[1] Frequency"=&RandomRangeFloat(Minimum=5, Maximum=7)
       "[1] Amplitude Min/Max"=<0.35 1>
      )
     )
    )
    Scale RGB=&MakeVectorFromLinearColorRGB(Color=&ColorFromCurve(ColorCurve=DataInterface() {propertyValues {
        AlphaCurve {keys []}
        BlueCurve {keys [<at 0 1 RCIM_Cubic>,<at 0.0601 1>,<at 0.0853 10>,<at 0.1113 1>,<at 0.1607 1>,<at 0.1834 0.2948>,<at 0.2047 1>,<at 0.3807 1>,<at 0.4058 0.3198>,<at 0.431 1>,<at 0.4931 1>,<at 0.519 20>,<at 0.5443 1>,<at 0.5919 1>,<at 0.6195 0.2906>,<at 0.6434 1>,<at 0.739 0.3781>,<at 0.7788 1>,<at 0.8684 1>,<at 0.9074 0.4156>,<at 0.9313 1>,<at 0.9702 1>]}
        GreenCurve {keys [<at 0 1 RCIM_Cubic>,<at 0.0601 1>,<at 0.0853 10>,<at 0.1113 1>,<at 0.1607 1>,<at 0.1834 0.2948>,<at 0.2047 1>,<at 0.3807 1>,<at 0.4058 0.3198>,<at 0.431 1>,<at 0.4931 1>,<at 0.519 20>,<at 0.5443 1>,<at 0.5919 1>,<at 0.6195 0.2906>,<at 0.6434 1>,<at 0.739 0.3781>,<at 0.7788 1>,<at 0.8684 1>,<at 0.9074 0.4156>,<at 0.9313 1>,<at 0.9702 1>]}
        RedCurve {keys [<at 0 1 RCIM_Cubic>,<at 0.0601 1>,<at 0.0853 10>,<at 0.1113 1>,<at 0.1607 1>,<at 0.1834 0.2948>,<at 0.2047 1>,<at 0.3807 1>,<at 0.4058 0.3198>,<at 0.431 1>,<at 0.4931 1>,<at 0.519 20>,<at 0.5443 1>,<at 0.5919 1>,<at 0.6195 0.2906>,<at 0.6434 1>,<at 0.739 0.3781>,<at 0.7788 1>,<at 0.8684 1>,<at 0.9074 0.4156>,<at 0.9313 1>,<at 0.9702 1>]}
       }}))
    ScaleA=true
    ScaleRGB=true
   )
   &SolveForcesAndVelocity(Clamp Velocity=false, Limit Acceleration=false)
   /SetVariables(
    Particles.Glow Color=&MultiplyLinearColor(
     Linear Color=&MakeLinearColorFromVectorAndFloat(
      "Vector (RGB)"=&MakeVectorFromLinearColorRGB(Color=&SelectLinearColorFromArray(
        Array Sampling Mode=::NewEnumerator0
        Direct Array Index=?
        Color Selection Array=DataInterface() {propertyValues {ColorData [<rgba 1 0.2143 0.7698 1>, <rgba 1 0.9076 0.3021 1>]}}
        Random Seed=0
       ))
      "Float (Alpha)"=0.8
     )
     Scale Factor=&MakeVector4_FromLinearColor(LinearColor=$Particles.Color)
    )
    Particles.Glow Size=&Multiply_Vector2D_ByFloat(Float=1.8, Vector2D=$Particles.SpriteSize)
   ) {isSetParameters=true}
  ]
  rendererClasses [&NiagaraSpriteRendererProperties]
  renderers [
   {rendererClass=&NiagaraSpriteRendererProperties}
   {rendererClass=&NiagaraSpriteRendererProperties, rendererIndex=1}
  ]
 }
}
summary {
 systemName=NS_Emote_ShinToll_PropSpawn
 emitters [
  {emitterName=NE_Particles, rendererClasses [&NiagaraSpriteRendererProperties]}
  {emitterName=NE_Burst, rendererClasses [&NiagaraSpriteRendererProperties, &NiagaraMeshRendererProperties]}
  {emitterName=NE_Ring, rendererClasses [&NiagaraSpriteRendererProperties]}
  {emitterName=NE_Heart, rendererClasses [&NiagaraSpriteRendererProperties]}
  {emitterName=NE_Sparkles, rendererClasses [&NiagaraSpriteRendererProperties]}
  {emitterName=NE_Light, rendererClasses [&NiagaraLightRendererProperties]}
 ]
 userVariables [
  User.bIsFrontEnd {type=NiagaraBool}
  User.bIsFrontEndPreview {type=NiagaraBool}
  User.FrontEndSize {type=NiagaraFloat, defaultValue=0.6, description=""}
  User.IsVariant {type=NiagaraFloat, defaultValue=0, description=""}
 ]
}
